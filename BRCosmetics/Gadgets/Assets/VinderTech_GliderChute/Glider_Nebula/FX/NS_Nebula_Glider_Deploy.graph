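# Niagara deploy effect for the Nebula glider, layering glows, sparks, body rings and spinning meshes positioned from the user skeletal mesh.
# Two moon-rock emitters spawn on spheres and get pulled inward, while ribbon flame emitters follow them, fading by source particle age.
# ParaGlideLeanAlpha raises flame spawn rate.
# Several emitters compute two variant color curves for paired renderers, some with alpha dimmed by time of day.
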NiagaraSystem /BRCosmetics/Gadgets/Assets/VinderTech_GliderChute/Glider_Nebula/FX/NS_Nebula_Glider_Deploy
dependencies {
 usedDataInterfaces [
  &NiagaraDataInterfaceCurve
  &NiagaraDataInterfaceColorCurve
  &NiagaraDataInterfaceParticleRead
  &NiagaraDataInterfaceArrayColor
 ]
 usedDynamicInputs [
  /Niagara/DynamicInputs/CascadeConversion/SplitVector
  &SetBoolByFloatComparison
  &SelectLinearColorFromArray
  &MakeIntFromFloat
 ]
 usedModules [
  &self:SystemSpawnScript.SystemScriptSource.SystemScriptGraph.NiagaraNodeAssignment_0.SetVariables_#1
  &self:SystemSpawnScript.SystemScriptSource.SystemScriptGraph.NiagaraNodeAssignment_1.SetVariables_#2
  &self:SystemSpawnScript.SystemScriptSource.SystemScriptGraph.NiagaraNodeAssignment_2.SetVariables_#3
  &self:SystemSpawnScript.SystemScriptSource.SystemScriptGraph.NiagaraNodeAssignment_3.SetVariables_#4
  &SystemState
 ]
}
emitters {
 Body_Rings002 {
  emitterUpdateScript [
   &EmitterState(Life Cycle Mode=::NewEnumerator1, Loop Behavior=::NewEnumerator1, Loop Delay=1.75, Loop Duration=0.75, UseLoopDelay=true)
   &SpawnBurst_Instantaneous(Use Spawn Probability=false, UseLoopCountLimit=false)
   &SpawnRate(
    SpawnRate=&FloatFromCurve(FloatCurve=DataInterface() {propertyValues {Curve {keys [<at 0 1>,<at 1 0>]}}}, CurveIndex=$Emitter.NormalizedLoopAge, Scale Curve=10)
    Use Spawn Probability=false
   )
  ]
  particleSpawnScript [
   &InitializeParticle(Mesh Scale Mode=::NewEnumerator3, Mesh Uniform Scale=1.5, UsePositionOffset=true, Write Material Random=true)
   &InitialMeshOrientation(Mesh Axis To Orient=<0 0 1>, Mesh Orientation Mode=::NewEnumerator2, Rotation=<0 0 0.25>, Use Rotation Vector=true)
  ]
  particleUpdateScript [
   /SetVariables(
    Particles.Variant0_ColorCurve:LinearColor=&MultiplyLinearColor(
     Linear Color=&ColorFromCurve(ColorCurve=DataInterface() {propertyValues {
        AlphaCurve {keys [<at 0 1 RCIM_Cubic>,<at 0.4012 1>,<at 0.6401 1>,<at 0.9957 0>]}
        BlueCurve {keys [<at 0 1 RCIM_Cubic>,<at 0.4894 3>,<at 0.6452 0.5906>,<at 1 0.5917 RCIM_Cubic>]}
        GreenCurve {keys [<at 0 0.3693 RCIM_Cubic>,<at 0.4894 0>,<at 0.6452 0.1058>,<at 1 0 RCIM_Cubic>]}
        RedCurve {keys [<at 0 0 RCIM_Cubic>,<at 0.4894 0.2427>,<at 0.6452 0>,<at 1 0.0292 RCIM_Cubic>]}
       }})
     Scale Factor=&MakeVector4(X=1, Y=1, Z=1, W=&FNTimeOfDayScalar(Night Scale=0.18, Morning Scale=1, Evening Scale=0.2))
    )
    Particles.Variant1_ColorCurve:LinearColor=&MultiplyLinearColor(
     Linear Color=&ColorFromCurve(ColorCurve=DataInterface() {propertyValues {
        AlphaCurve {keys [<at 0 1>,<at 0.4243 1>,<at 1 0>]}
        BlueCurve {keys [<at 0.0987 0.1912>,<at 0.2922 0.0569>,<at 0.6383 1>,<at 0.9715 0.1912>]}
        GreenCurve {keys [<at 0.0987 0>,<at 0.2922 0>,<at 0.6383 0.0021>,<at 0.9715 0>]}
        RedCurve {keys [<at 0.0987 0.0037>,<at 0.2922 1>,<at 0.6383 0>,<at 0.9715 0.0037>]}
       }})
     Scale Factor=&MakeVector4(X=1, Y=1, Z=1, W=&FNTimeOfDayScalar(Night Scale=0.18, Morning Scale=0.5, Evening Scale=0.18))
    )
   ) {isSetParameters=true}
   &ParticleState()
   &SkeletalMeshLocation(Bone Sampling Mode=::NewEnumerator4, MaskSpawnGroup=false, Sampled Position Offset=<0 15 0>, Skeletal Mesh=$User.SkeletalMesh)
   &ScaleMeshSize(Scale Factor=&VectorFromFloat(Value=&FloatFromCurve(
      FloatCurve=DataInterface() {propertyValues {Curve {keys [{tangentMode=RCTM_User, time=0, value=0}, {tangentMode=RCTM_User, time=1, value=1}]}}}
      Scale Curve=2
     ))) {module=ScaleMeshSize001}
   &SolveForcesAndVelocity(Clamp Velocity=false, Limit Acceleration=false)
   &ScaleColor(
    Linear Color Curve=DataInterface() {propertyValues {
      AlphaCurve {keys [<at 0 1 RCIM_Cubic>,<at 0.4012 0.75>,<at 0.9957 0>]}
      BlueCurve {keys [<at 0 0.5906 RCIM_Cubic>,<at 0.4894 3>,<at 0.6452 0.5906>,<at 1 0.5917 RCIM_Cubic>]}
      GreenCurve {keys [<at 0 0 RCIM_Cubic>,<at 0.4894 0>,<at 0.6452 0.1058>,<at 1 0 RCIM_Cubic>]}
      RedCurve {keys [<at 0 0.0296 RCIM_Cubic>,<at 0.4894 0.2427>,<at 0.6452 0>,<at 1 0.0292 RCIM_Cubic>]}
     }}
    Scale Mode=::NewEnumerator2
   )
  ]
  rendererClasses [&NiagaraMeshRendererProperties]
  renderers [
   {rendererClass=&NiagaraMeshRendererProperties, rendererIndex=0}
   {rendererClass=&NiagaraMeshRendererProperties, rendererIndex=1}
  ]
 }
 CenterGlow {
  emitterUpdateScript [
   &EmitterState(Life Cycle Mode=::NewEnumerator1, Loop Behavior=::NewEnumerator1, Loop Duration=3, UseLoopDelay=false)
   &SpawnBurst_Instantaneous(Use Spawn Probability=false, UseLoopCountLimit=false)
  ]
  particleSpawnScript [
   /SetVariables(Particles.Color_Sprite:LinearColor=$System.GlowColor, Particles.Size_Sprite:Vector2f=<50 50>) {isSetParameters=true}
   &InitializeParticle(
    Lifetime=3
    Mesh Scale Mode=::NewEnumerator3
    Mesh Uniform Scale=0.25
    Sprite Size Mode=::NewEnumerator3
    Uniform Sprite Size=150
    UsePositionOffset=false
    Write Material Random=true
   )
  ]
  particleUpdateScript [
   &ParticleState(
    Kill Particles When Lifetime Has Elapsed=false
    Let Infinitely Lived Particles Die When Emitter Deactivates=true
    Loop Particles Lifetime=true
   )
   &SkeletalMeshLocation(Bone / Socket Index=$Particles.SpawnGroup, Bone Sampling Mode=::NewEnumerator4, MaskSpawnGroup=false, Skeletal Mesh=$User.SkeletalMesh)
   &ScaleSpriteSize(
    Uniform Curve Scale=1.5
    Uniform Curve Sprite Scale=DataInterface() {propertyValues {Curve {keys [
        {interpMode=RCIM_Cubic, tangentMode=RCTM_User, time=0, value=0}
        {interpMode=RCIM_Cubic, tangentMode=RCTM_User, time=0.3, value=1}
        <at 1 0.5 RCIM_Cubic>
       ]}}}
   )
   &SolveForcesAndVelocity(Clamp Velocity=false, Limit Acceleration=false)
   &ScaleColor(
    Scale Alpha=&FloatFromCurve(
     FloatCurve=DataInterface() {propertyValues {Curve {keys [
         {tangentMode=RCTM_User, time=0, value=1}
         {arriveTangent=-0.0405, interpMode=RCIM_Cubic, leaveTangent=-0.0405, tangentMode=RCTM_User, time=0.7009, value=1}
         {arriveTangent=-8.0089, interpMode=RCIM_Cubic, leaveTangent=-8.0089, tangentMode=RCTM_User, time=1, value=0}
        ]}}}
     Scale Curve=5
    )
    ScaleA=true
    ScaleRGB=true
   )
   &CameraOffset(Camera Offset Amount=25)
  ]
  rendererClasses [&NiagaraSpriteRendererProperties]
  renderers [{rendererClass=&NiagaraSpriteRendererProperties}]
 }
 Flames_Attached_to_Rocks001 {
  emitterUpdateScript [
   &EmitterState()
   &SpawnParticlesFromOtherEmitter(
    Attribute Reader=DataInterface() {propertyValues {EmitterBinding {bindingMode=Other, emitterName=Rock_Moon_02001}}}
    Spawn Rate=&Lerp_Float(A=150, B=300, Alpha=$User.ParaGlideLeanAlpha)
    Spawn Rate Per Particle Cap=10000000
    Use Maximum Age=false
    Use Minimum Age=false
   )
  ]
  particleSpawnScript [
   &InitializeParticle(UsePositionOffset=false, Write Material Random=true)
   &InitializeRibbon(
    Color=$System.GlowColor
    Lifetime=0.25
    Position Mode:ENiagara_PositionInitializationMode=::NewEnumerator0
    Ribbon Facing Mode:ENiagara_UnsetDirectSet=::NewEnumerator1
    Ribbon Facing Vector:Vector3f=<1 0 0>
    Ribbon Width=45
    Write Material Random=true
   )
   &SampleParticlesFromOtherEmitter(
    Applied Position Offset:Vector3f=<0 0 -10>
    Applied Velocity Scale=0
    Particle ID Sampling=::NewEnumerator1
    Sprite Size Sampling:ENiagara_DirectReadSamplingMode=::NewEnumerator1
    Velocity Sampling=::NewEnumerator0
   )
   /SetVariables(Particles.RibbonWidth=&Multiply_Float(A=&MakeFloatFromVector2d(Vector 2D=$Output.SampleParticlesFromOtherEmitter.SampledSpriteSize), B=1.5)) {isSetParameters=true, module=SetVariables_#5}
   &AddVelocity(Velocity=<0 0 250>)
  ]
  particleUpdateScript [
   &InheritSourceMovement(
    Applied Position Scale=<-0.15 -0.15 0>
    ApplyForce=false
    ApplyPosition=true
    ApplyPositionFalloff=false
    Source Velocity:Vector3f=&TransformVector(Vector=$Engine.Owner.Velocity, Source Space=::World, Destination Space=::Simulation)
   )
   &self:ReadRibbonSourceParticleAge_05(PAR=DataInterface() {propertyValues {EmitterBinding {bindingMode=Other, emitterName=Rock_Moon_02001}}}) {module=ReadRibbonSourceParticleAge_05}
   &SolveForcesAndVelocity(Clamp Velocity=false, Limit Acceleration=false)
   &ScaleColor(
    Scale Alpha=&FloatFromCurve(FloatCurve=DataInterface() {propertyValues {Curve {keys [
         {interpMode=RCIM_Cubic, tangentMode=RCTM_User, time=0, value=0}
         {interpMode=RCIM_Cubic, tangentMode=RCTM_User, time=0.5, value=1}
         {tangentMode=RCTM_User, time=1, value=0}
        ]}}})
    ScaleA=true
    ScaleRGB=true
   ) {module=ScaleColor001}
   &DynamicMaterialParameters(
    Index 1 Param 4=&Lerp_Float(
     A=0
     B=1
     Alpha=&FloatFromCurve(
      FloatCurve=DataInterface() {propertyValues {Curve {keys [
          {interpMode=RCIM_Cubic, tangentMode=RCTM_User, time=0.25, value=0}
          {arriveTangent=4.6598, interpMode=RCIM_Cubic, leaveTangent=4.6598, tangentMode=RCTM_User, time=0.75, value=1}
         ]}}}
      CurveIndex=$Particles.SourceParticleAge
     )
    )
    Module.Param0WriteEnabled1=false
    Module.Param1WriteEnabled1=false
    Module.Param2WriteEnabled1=false
    Module.Param3WriteEnabled1=true
    Write Parameter Index 0=false
    Write Parameter Index 1=true
   ) {module=DynamicMaterialParameters001}
   &ParticleState()
   &ScaleRibbonWidth(Ribbon Width Scale=&FloatFromCurve(
     FloatCurve=DataInterface() {propertyValues {Curve {keys [
         {interpMode=RCIM_Cubic, tangentMode=RCTM_User, time=0, value=0}
         {interpMode=RCIM_Cubic, tangentMode=RCTM_User, time=0.5, value=1}
         {tangentMode=RCTM_User, time=1, value=0}
        ]}}}
     CurveIndex=$Particles.SourceParticleAge
    ))
   &ScaleColor(
    Scale Alpha=&FloatFromCurve(
     FloatCurve=DataInterface() {propertyValues {Curve {keys [
         {interpMode=RCIM_Cubic, tangentMode=RCTM_User, time=0, value=0}
         {interpMode=RCIM_Cubic, tangentMode=RCTM_User, time=0.5, value=1}
         {tangentMode=RCTM_User, time=1, value=0}
        ]}}}
     CurveIndex=$Particles.SourceParticleAge
    )
    ScaleA=true
    ScaleRGB=true
   )
  ]
  rendererClasses [&NiagaraRibbonRendererProperties]
  renderers [{rendererClass=&NiagaraRibbonRendererProperties}]
 }
 Flames_Attached_to_Rocks007 {
  emitterUpdateScript [
   &EmitterState()
   &SpawnParticlesFromOtherEmitter(
    Attribute Reader=DataInterface() {propertyValues {EmitterBinding {bindingMode=Other, emitterName=Rock_Moon_02}}}
    Spawn Rate=&Lerp_Float(A=150, B=300, Alpha=$User.ParaGlideLeanAlpha)
    Spawn Rate Per Particle Cap=10000000
    Use Maximum Age=false
    Use Minimum Age=false
   )
  ]
  particleSpawnScript [
   &InitializeParticle(UsePositionOffset=false, Write Material Random=true)
   &InitializeRibbon(
    Color=&RandomRangeLinearColor(Minimum=<rgba 0 0 0 1>, Maximum=<rgba 1 1 1 1>)
    Lifetime Max=0.3
    Lifetime Min=0.25
    Lifetime Mode:ENiagara_LifetimeMode=::NewEnumerator1
    Position Mode:ENiagara_PositionInitializationMode=::NewEnumerator0
    Ribbon Facing Mode:ENiagara_UnsetDirectSet=::NewEnumerator1
    Ribbon Facing Vector:Vector3f=<1 0 0>
    Ribbon Width=30
    Write Material Random=true
   )
   &SampleParticlesFromOtherEmitter(
    Applied Position Offset:Vector3f=<0 0 -10>
    Applied Velocity Scale=0
    Particle ID Sampling=::NewEnumerator1
    Sprite Size Sampling:ENiagara_DirectReadSamplingMode=::NewEnumerator1
    Velocity Sampling=::NewEnumerator0
   )
   /SetVariables(Particles.RibbonWidth=&Multiply_Float(A=&MakeFloatFromVector2d(Vector 2D=$Output.SampleParticlesFromOtherEmitter.SampledSpriteSize), B=1.5)) {isSetParameters=true, module=SetVariables_#5}
   &AddVelocity(Velocity=<0 0 250>)
  ]
  particleUpdateScript [
   /SetVariables(
    Particles.Variant0_ColorCurve:LinearColor=&MultiplyLinearColor(
     Linear Color=&ColorFromCurve(ColorCurve=$System.Variant0_ColorCurve)
     Scale Factor=/Niagara/DynamicInputs/UniformRange/V2/RandomRangeVector4(
      Minimum:Vector4f=<1 1 1 0.5>
      Maximum:Vector4f=<5 5 5 2>
      Evaluation Type:ENiagaraRandomnessEvaluation=::NewEnumerator0
      Recalculate Random Each Loop:NiagaraBool=true
      Randomness Mode:ENiagaraRandomnessMode=::NewEnumerator0
      Random Seed:NiagaraInt32=0
      Fixed Random Seed:NiagaraBool=false
     )
    )
    Particles.Variant1_ColorCurve:LinearColor=&MultiplyLinearColor(
     Linear Color=&ColorFromCurve(ColorCurve=$System.Variant0_ColorCurve)
     Scale Factor=&RandomRangeVector4(
      Minimum:Vector4f=<1 1 1 0.5>
      Maximum:Vector4f=<5 5 5 2>
      Evaluation Type:ENiagaraRandomnessEvaluation=::NewEnumerator0
      Recalculate Random Each Loop:NiagaraBool=true
      Randomness Mode:ENiagaraRandomnessMode=::NewEnumerator0
      Random Seed:NiagaraInt32=0
      Fixed Random Seed:NiagaraBool=false
     )
    )
   ) {isSetParameters=true}
   &InheritSourceMovement(
    Applied Position Scale=<-0.15 -0.15 0>
    ApplyForce=false
    ApplyPosition=true
    ApplyPositionFalloff=false
    Source Velocity:Vector3f=&TransformVector(Vector=$Engine.Owner.Velocity, Source Space=::World, Destination Space=::Simulation)
   )
   &self:ReadRibbonSourceParticleAge_05(PAR=DataInterface() {propertyValues {EmitterBinding {bindingMode=Other, emitterName=Rock_Moon_02}}}) {module=ReadRibbonSourceParticleAge_05}
   &SolveForcesAndVelocity(Clamp Velocity=false, Limit Acceleration=false)
   &DynamicMaterialParameters(
    Index 1 Param 4=&Lerp_Float(
     A=0
     B=1
     Alpha=&FloatFromCurve(
      FloatCurve=DataInterface() {propertyValues {Curve {keys [
          {interpMode=RCIM_Cubic, tangentMode=RCTM_User, time=0.25, value=0}
          {arriveTangent=4.6598, interpMode=RCIM_Cubic, leaveTangent=4.6598, tangentMode=RCTM_User, time=0.75, value=1}
         ]}}}
      CurveIndex=$Particles.SourceParticleAge
     )
    )
    Module.Param0WriteEnabled1=false
    Module.Param1WriteEnabled1=false
    Module.Param2WriteEnabled1=false
    Module.Param3WriteEnabled1=true
    Write Parameter Index 0=false
    Write Parameter Index 1=true
   ) {module=DynamicMaterialParameters001}
   &ParticleState()
   &ScaleRibbonWidth(Ribbon Width Scale=&FloatFromCurve(
     FloatCurve=DataInterface() {propertyValues {Curve {keys [{tangentMode=RCTM_User, time=0, value=1}, {tangentMode=RCTM_User, time=1, value=0}]}}}
     CurveIndex=$Particles.SourceParticleAge
    ))
   &ScaleColor(
    Scale Alpha=&FloatFromCurve(
     FloatCurve=DataInterface() {propertyValues {Curve {keys [
         {interpMode=RCIM_Cubic, tangentMode=RCTM_User, time=0, value=0}
         {interpMode=RCIM_Cubic, tangentMode=RCTM_User, time=0.5, value=1}
         {tangentMode=RCTM_User, time=1, value=0}
        ]}}}
     CurveIndex=$Particles.SourceParticleAge
    )
    ScaleA=true
    ScaleRGB=true
   )
  ]
  rendererClasses [&NiagaraRibbonRendererProperties]
  renderers [
   {rendererClass=&NiagaraRibbonRendererProperties}
   {rendererClass=&NiagaraRibbonRendererProperties, rendererIndex=1}
  ]
 }
 Glow {
  emitterUpdateScript [
   &EmitterState()
   &SpawnBurst_Instantaneous(Use Spawn Probability=false, UseLoopCountLimit=false)
  ]
  particleSpawnScript [&InitializeParticle(
    Color=<rgba 1 1 1 0.5>
    Lifetime=2
    Sprite Size Mode=::NewEnumerator3
    Uniform Sprite Size=500
    UsePositionOffset=false
    Write Material Random=true
   )]
  particleUpdateScript [
   /SetVariables(
    Particles.Variant0_ColorCurve:LinearColor=&MultiplyLinearColor(
     Linear Color=&ColorFromCurve(ColorCurve=$System.Variant0_ColorCurve_Energy)
     Scale Factor=&MakeVector4(X=1, Y=1, Z=1, W=&FNTimeOfDayScalar(Night Scale=0.3, Morning Scale=1, Evening Scale=0.3))
    )
    Particles.Variant1_ColorCurve:LinearColor=&MultiplyLinearColor(
     Linear Color=&ColorFromCurve(ColorCurve=$System.Variant1_ColorCurve_Energy)
     Scale Factor=&MakeVector4(X=1, Y=1, Z=1, W=&FNTimeOfDayScalar(Night Scale=0.3, Morning Scale=1, Evening Scale=0.3))
    )
   ) {isSetParameters=true}
   &ParticleState()
   &ScaleSpriteSize(Uniform Curve Sprite Scale=DataInterface() {propertyValues {Curve {keys [{interpMode=RCIM_Cubic, tangentMode=RCTM_User, time=0, value=1}, {interpMode=RCIM_Cubic, tangentMode=RCTM_User, time=1, value=0}]}}})
   &SkeletalMeshLocation(Bone Sampling Mode=::NewEnumerator4, MaskSpawnGroup=false, Skeletal Mesh=$User.SkeletalMesh)
   &CameraOffset(Camera Offset Amount=&FloatFromCurve(FloatCurve=DataInterface() {propertyValues {Curve {keys [<at 0 1>,<at 1 0>]}}}, Scale Curve=50))
  ]
  rendererClasses [&NiagaraSpriteRendererProperties]
  renderers [
   {rendererClass=&NiagaraSpriteRendererProperties}
   {rendererClass=&NiagaraSpriteRendererProperties, rendererIndex=1}
  ]
 }
 Rock_Moon_02 {
  emitterUpdateScript [
   &EmitterState(Life Cycle Mode=::NewEnumerator1, Loop Behavior=::NewEnumerator1, Loop Duration=1.5, UseLoopDelay=false)
   &SpawnRate(
    SpawnRate=&FloatFromCurve(FloatCurve=DataInterface() {propertyValues {Curve {keys [<at 0 1>,<at 1 0>]}}}, CurveIndex=$Emitter.NormalizedLoopAge, Scale Curve=45)
    Use Spawn Probability=false
   )
   &SpawnBurst_Instantaneous(
    Spawn Count=&RandomRangeInt(Minimum=5, Maximum=9)
    Spawn Time=0.3
    Use Spawn Probability=false
    UseLoopCountLimit=false
   ) {module=SpawnBurst_Instantaneous002}
   &SpawnBurst_Instantaneous(
    Spawn Count=&RandomRangeInt(Minimum=5, Maximum=9)
    Spawn Time=0.45
    Use Spawn Probability=false
    UseLoopCountLimit=false
   ) {module=SpawnBurst_Instantaneous001}
  ]
  particleSpawnScript [
   &InitializeParticle(
    Color Mode=::NewEnumerator0
    Lifetime Max=0.7
    Lifetime Min=0.3
    Lifetime Mode=::NewEnumerator1
    Mesh Scale Mode=::NewEnumerator1
    Mesh Uniform Scale Max=0.5
    Mesh Uniform Scale Min=0.3
    UsePositionOffset=false
    Write Material Random=true
   )
   /SetVariables(Particles.Color_Sprite:LinearColor=$System.RockGlowColor, Particles.Size_Sprite:Vector2f=<27 27>) {isSetParameters=true, module=SetVariables_#6}
   &ShapeLocation(
    MaskSpawnGroup=false
    Override Seed=false
    Sphere Radius=&RandomRangeFloat(Minimum=300, Maximum=500)
    UseOwnerScale=true
   )
   &InitialMeshOrientation(Use Rotation Vector=false)
   &SkeletalMeshLocation(
    Bone Sampling Mode=::NewEnumerator4
    MaskSpawnGroup=false
    Orientation Sampling=::NewEnumerator1
    Skeletal Mesh=$User.SkeletalMesh
    Write To Particle Position=false
   ) {enabled=false}
   /SetVariables(Particles.SpawnPosition:NiagaraPosition=&AddVectorToPosition(
     Position=$Particles.SkeletalMeshLocation.SampledPosition
     Vector=&Lerp_Vector(
      A=&RandomRangeVector(Minimum=<-25 10 100>, Maximum=<25 25 150>)
      B=&RandomRangeVector(Minimum=<-50 100 100>, Maximum=<50 50 150>)
      Alpha=$System.Dive
     )
    )) {enabled=false, isSetParameters=true, module=SetVariables_#7}
  ]
  particleUpdateScript [
   &PointAttractionForce(
    Attraction Radius=500
    Attraction Strength=&FloatFromCurve(FloatCurve=DataInterface() {propertyValues {Curve {keys [<at 0 1>,<at 1 0>]}}}, Scale Curve=300)
    Attractor Position Offset:Vector3f=<0 0 -25>
    Kill Radius=1
    Kill Within Radius=true
    Module.Use Falloff=true
    UseAttractorOffset=true
   )
   &SolveForcesAndVelocity(Clamp Velocity=false, Limit Acceleration=false)
   &SkeletalMeshLocation(
    Bone Sampling Mode=::NewEnumerator4
    MaskSpawnGroup=false
    Orientation Sampling=::NewEnumerator1
    Sampled Position Offset Coordinate Space:ENiagaraSkelMeshTransforms=::NewEnumerator2
    Skeletal Mesh=$User.SkeletalMesh
    Write To Particle Position=false
   ) {enabled=false, module=SkeletalMeshLocation001}
   /SetVariables(Particles.Position=&LerpPosition(
     A=$Particles.SpawnPosition
     B=$Particles.SkeletalMeshLocation001.SampledPosition
     Alpha=&Add_Float(A=$Particles.NormalizedAge, B=&RandomRangeFloat(Minimum=-0.15, Maximum=0))
    )) {enabled=false, isSetParameters=true, module=SetVariables_#8}
   &ParticleState()
   &UpdateMeshOrientation(
    Rotation Rate=&Multiply_Float(
     A=&Divide_Float(A=5, B=$Particles.Mass)
     B=&RandomRangeFloat(Minimum=0.3, Maximum=0.5)
    )
    Rotation Vector=&RandomVector(Vector Scale=1)
   )
   &ScaleColor(
    Color Value To Scale=$Particles.Initial.Color_Sprite
    Linear Color Curve=DataInterface() {propertyValues {
      AlphaCurve {keys [<at 0.2958 0>,<at 0.5459 1>,<at 1 0>]}
      BlueCurve {keys [<at 0 1>,<at 0.2649 1>,<at 1 1>]}
      GreenCurve {keys [<at 0 0.54>,<at 0.2649 1>,<at 1 0>]}
      RedCurve {keys [<at 0 0>,<at 0.2649 1>,<at 1 0.509>]}
     }}
    Scale Mode=::NewEnumerator2
   )
   &ScaleSpriteSize(
    Initial Sprite Size:Vector2f=$Particles.Initial.Size_Sprite
    Uniform Curve Scale=1.5
    Uniform Curve Sprite Scale=DataInterface() {propertyValues {Curve {keys [
        {arriveTangent=-0.0548, interpMode=RCIM_Cubic, leaveTangent=-0.0548, tangentMode=RCTM_User, time=0, value=0}
        {interpMode=RCIM_Cubic, tangentMode=RCTM_User, time=0.5, value=1}
        {arriveTangent=-5.1386, interpMode=RCIM_Cubic, leaveTangent=-5.1386, tangentMode=RCTM_User, time=1, value=0}
       ]}}}
   )
   &ScaleMeshSize(Scale Factor=&Multiply_VectorByFloat(
     Vector=$Engine.Owner.Scale
     Float=&FloatFromCurve(
      FloatCurve=DataInterface() {propertyValues {Curve {keys [
          <at 0 0 RCIM_Cubic>
          {arriveTangent=0.3333, interpMode=RCIM_Cubic, leaveTangent=0.3333, time=0.2, value=0.5}
          {arriveTangent=-0.2898, interpMode=RCIM_Cubic, leaveTangent=-0.2898, tangentMode=RCTM_User, time=0.9, value=0.3}
          <at 1 0.3>
         ]}}}
      Scale Curve=1.5
     )
    ))
   &GenerateDeathEvent(Use Event Probability=false)
  ]
  rendererClasses [&NiagaraMeshRendererProperties, &NiagaraSpriteRendererProperties]
  renderers [
   {rendererClass=&NiagaraMeshRendererProperties, rendererIndex=0}
   {rendererClass=&NiagaraSpriteRendererProperties, rendererIndex=1}
  ]
 }
 Rock_Moon_02001 {
  emitterUpdateScript [
   &EmitterState(Life Cycle Mode=::NewEnumerator1, Loop Behavior=::NewEnumerator1, Loop Duration=1.5, UseLoopDelay=false)
   &SpawnRate(
    SpawnRate=&FloatFromCurve(FloatCurve=DataInterface() {propertyValues {Curve {keys [<at 0 1>,<at 1 0>]}}}, CurveIndex=$Emitter.NormalizedLoopAge, Scale Curve=15)
    Use Spawn Probability=false
   )
  ]
  particleSpawnScript [
   &InitializeParticle(
    Color Mode=::NewEnumerator0
    Lifetime Max=0.7
    Lifetime Min=0.3
    Lifetime Mode=::NewEnumerator1
    Mesh Scale Mode=::NewEnumerator1
    Mesh Uniform Scale Max=0.1
    Mesh Uniform Scale Min=0.05
    Sprite Size Mode=::NewEnumerator3
    UsePositionOffset=false
    Write Material Random=true
   )
   /SetVariables(Particles.Color_Sprite:LinearColor=$System.GlowColor, Particles.Size_Sprite:Vector2f=<17 17>) {isSetParameters=true, module=SetVariables_#6}
   &ShapeLocation(
    MaskSpawnGroup=false
    Override Seed=false
    Sphere Radius=&RandomRangeFloat(Minimum=300, Maximum=500)
    UseOwnerScale=true
   )
   &InitialMeshOrientation(Use Rotation Vector=false)
   &SkeletalMeshLocation(
    Bone Sampling Mode=::NewEnumerator4
    MaskSpawnGroup=false
    Orientation Sampling=::NewEnumerator1
    Skeletal Mesh=$User.SkeletalMesh
    Write To Particle Position=false
   ) {enabled=false}
   /SetVariables(Particles.SpawnPosition:NiagaraPosition=&AddVectorToPosition(
     Position=$Particles.SkeletalMeshLocation.SampledPosition
     Vector=&Lerp_Vector(
      A=&RandomRangeVector(Minimum=<-25 10 100>, Maximum=<25 25 150>)
      B=&RandomRangeVector(Minimum=<-50 100 100>, Maximum=<50 50 150>)
      Alpha=$System.Dive
     )
    )) {enabled=false, isSetParameters=true, module=SetVariables_#7}
  ]
  particleUpdateScript [
   &PointAttractionForce(
    Attraction Radius=500
    Attraction Strength=&FloatFromCurve(FloatCurve=DataInterface() {propertyValues {Curve {keys [<at 0 1>,<at 1 0>]}}}, Scale Curve=300)
    Attractor Position Offset:Vector3f=<0 0 -25>
    Kill Radius=1
    Kill Within Radius=true
    Module.Use Falloff=true
    UseAttractorOffset=true
   )
   &SolveForcesAndVelocity(Clamp Velocity=false, Limit Acceleration=false)
   &SkeletalMeshLocation(
    Bone Sampling Mode=::NewEnumerator4
    MaskSpawnGroup=false
    Orientation Sampling=::NewEnumerator1
    Sampled Position Offset Coordinate Space:ENiagaraSkelMeshTransforms=::NewEnumerator2
    Skeletal Mesh=$User.SkeletalMesh
    Write To Particle Position=false
   ) {enabled=false, module=SkeletalMeshLocation001}
   /SetVariables(Particles.Position=&LerpPosition(
     A=$Particles.SpawnPosition
     B=$Particles.SkeletalMeshLocation001.SampledPosition
     Alpha=&Add_Float(A=$Particles.NormalizedAge, B=&RandomRangeFloat(Minimum=-0.15, Maximum=0))
    )) {enabled=false, isSetParameters=true, module=SetVariables_#8}
   &ParticleState()
   &UpdateMeshOrientation(
    Rotation Rate=&Multiply_Float(
     A=&Divide_Float(A=5, B=$Particles.Mass)
     B=&RandomRangeFloat(Minimum=0.3, Maximum=0.5)
    )
    Rotation Vector=&RandomVector(Vector Scale=1)
   )
   &ScaleColor(
    Color Value To Scale=$Particles.Initial.Color_Sprite
    Linear Color Curve=DataInterface() {propertyValues {
      AlphaCurve {keys [<at 0.2958 0>,<at 0.5459 1>,<at 1 0>]}
      BlueCurve {keys [<at 0 1>,<at 0.2649 1>,<at 1 1>]}
      GreenCurve {keys [<at 0 0.54>,<at 0.2649 1>,<at 1 0>]}
      RedCurve {keys [<at 0 0>,<at 0.2649 1>,<at 1 0.509>]}
     }}
    Scale Mode=::NewEnumerator2
   )
   &ScaleSpriteSize(
    Initial Sprite Size:Vector2f=$Particles.Initial.Size_Sprite
    Uniform Curve Scale=1.5
    Uniform Curve Sprite Scale=DataInterface() {propertyValues {Curve {keys [
        {arriveTangent=-0.0548, interpMode=RCIM_Cubic, leaveTangent=-0.0548, tangentMode=RCTM_User, time=0, value=0}
        {interpMode=RCIM_Cubic, tangentMode=RCTM_User, time=0.5, value=1}
        {arriveTangent=-5.1386, interpMode=RCIM_Cubic, leaveTangent=-5.1386, tangentMode=RCTM_User, time=1, value=0}
       ]}}}
   )
   &ScaleMeshSize(Scale Factor=&Multiply_VectorByFloat(
     Vector=$Engine.Owner.Scale
     Float=&FloatFromCurve(
      FloatCurve=DataInterface() {propertyValues {Curve {keys [
          <at 0 0 RCIM_Cubic>
          {arriveTangent=0.3333, interpMode=RCIM_Cubic, leaveTangent=0.3333, time=0.2, value=0.5}
          {arriveTangent=-0.2898, interpMode=RCIM_Cubic, leaveTangent=-0.2898, tangentMode=RCTM_User, time=0.9, value=0.3}
          <at 1 0.3>
         ]}}}
      Scale Curve=1.5
     )
    ))
   &GenerateDeathEvent(Use Event Probability=false)
  ]
  rendererClasses [&NiagaraMeshRendererProperties, &NiagaraSpriteRendererProperties]
  renderers [
   {rendererClass=&NiagaraMeshRendererProperties, rendererIndex=0}
   {rendererClass=&NiagaraSpriteRendererProperties, rendererIndex=1}
  ]
 }
 Sparks {
  emitterUpdateScript [
   &EmitterState(Life Cycle Mode=::NewEnumerator1, Loop Behavior=::NewEnumerator1, Loop Duration=1, UseLoopDelay=false)
   &SpawnRate(
    SpawnRate=&FloatFromCurve(
     FloatCurve=DataInterface() {propertyValues {Curve {keys [{interpMode=RCIM_Cubic, tangentMode=RCTM_User, time=0, value=1}, {interpMode=RCIM_Cubic, tangentMode=RCTM_User, time=0.4, value=0}]}}}
     CurveIndex=$Emitter.NormalizedLoopAge
     Scale Curve=100
    )
    Use Spawn Probability=false
   )
  ]
  particleSpawnScript [
   &InitializeParticle(
    Color Maximum=<rgba 1 0 0.3654 1>
    Color Minimum=<rgba 0 0.3899 1 1>
    Color Mode=::NewEnumerator2
    Sprite Size Mode=::NewEnumerator1
    Uniform Sprite Size Max=3
    Uniform Sprite Size Min=1
    UsePositionOffset=false
    Write Material Random=true
   )
   &SkeletalMeshLocation(Bone Sampling Mode=::NewEnumerator4, MaskSpawnGroup=false, Skeletal Mesh=$User.SkeletalMesh)
   &ShapeLocation(
    MaskSpawnGroup=false
    Override Seed=false
    Shape Origin=$Particles.SkeletalMeshLocation.SampledPosition
    Sphere Radius=400
    UseOwnerScale=true
   )
  ]
  particleUpdateScript [
   /SetVariables(
    Particles.Variant0_ColorCurve:LinearColor=&ColorFromCurve(ColorCurve=DataInterface() {propertyValues {
       AlphaCurve {keys [
         <at 0.0038 0>
         {arriveTangent=1.8872, interpMode=RCIM_Cubic, leaveTangent=1.8872, time=0.0763, value=1}
         <at 0.4012 0.75>
         <at 0.8374 1>
         <at 0.9957 0>
        ]}
       BlueCurve {keys [<at 0 1 RCIM_Cubic>,<at 0.4894 3>,<at 0.6452 0.5906>,<at 1 0.5917 RCIM_Cubic>]}
       GreenCurve {keys [<at 0 0 RCIM_Cubic>,<at 0.4894 0>,<at 0.6452 0.1058>,<at 1 0 RCIM_Cubic>]}
       RedCurve {keys [<at 0 0.05 RCIM_Cubic>,<at 0.4894 0.2427>,<at 0.6452 0>,<at 1 0.0292 RCIM_Cubic>]}
      }})
    Particles.Variant1_ColorCurve:LinearColor=&MultiplyLinearColor(
     Linear Color=&ColorFromCurve(ColorCurve=DataInterface() {propertyValues {
        AlphaCurve {keys [<at 0.0039 0>,<at 0.0695 1>,<at 0.4464 0.85>,<at 0.719 1>,<at 1 0>]}
        BlueCurve {keys [<at 0 1>,<at 0.2996 1>,<at 0.4722 0.0441>,<at 0.9715 0.0194>]}
        GreenCurve {keys [<at 0 0.3088>,<at 0.2996 0>,<at 0.4722 0>,<at 0.9715 0>]}
        RedCurve {keys [<at 0 0.0552>,<at 0.2996 0.0195>,<at 0.4722 1>,<at 0.9715 0.1958>]}
       }})
     Scale Factor=&MakeVector4(X=1, Y=1, Z=1, W=&FNTimeOfDayScalar(Night Scale=0.5, Morning Scale=1, Evening Scale=0.5))
    )
   ) {isSetParameters=true}
   &ParticleState()
   &ScaleSpriteSize(Uniform Curve Sprite Scale=DataInterface() {propertyValues {Curve {keys [
        {arriveTangent=2.5396, interpMode=RCIM_Cubic, leaveTangent=2.5396, tangentMode=RCTM_User, time=0, value=0}
        {interpMode=RCIM_Cubic, tangentMode=RCTM_User, time=1, value=1}
       ]}}})
   &PointAttractionForce(
    Attraction Radius=400
    Attraction Strength=50
    Attractor Position:NiagaraPosition=$Particles.SkeletalMeshLocation.SampledPosition
    Kill Radius=1
    Kill Within Radius=false
    Module.Use Falloff=true
    UseAttractorOffset=false
   )
   &SolveForcesAndVelocity(Clamp Velocity=false, Limit Acceleration=false)
   &ScaleSpriteSizeBySpeed(Max Scale Factor=<1 10>)
   &ScaleColor(
    Scale Alpha=&FloatFromCurve(FloatCurve=DataInterface() {propertyValues {Curve {keys [{interpMode=RCIM_Cubic, tangentMode=RCTM_User, time=0, value=0}, {interpMode=RCIM_Cubic, tangentMode=RCTM_User, time=1, value=1}]}}})
    ScaleA=true
    ScaleRGB=true
   )
  ]
  rendererClasses [&NiagaraSpriteRendererProperties]
  renderers [
   {rendererClass=&NiagaraSpriteRendererProperties}
   {rendererClass=&NiagaraSpriteRendererProperties, rendererIndex=1}
  ]
 }
 SpinnyStatic {
  emitterUpdateScript [
   &EmitterState()
   &SpawnBurst_Instantaneous(Spawn Count=1, Use Spawn Probability=false)
  ]
  particleSpawnScript [
   &InitializeParticle(Mesh Scale Mode=::NewEnumerator3, Mesh Uniform Scale=4, Position Offset=<8 0 0>, UsePositionOffset=true, Write Material Random=true) {module=InitializeParticle001}
   &InitialMeshOrientation(
    Rotation=&RandomRangeVector(Minimum=<-1 0 0>, Maximum=<1 0 0>)
    Use Orientation Vector=true
    Use Rotation Vector=true
   )
  ]
  particleUpdateScript [
   /SetVariables(
    Particles.Variant0_ColorCurve:LinearColor=&ColorFromCurve(ColorCurve=DataInterface() {propertyValues {
       AlphaCurve {keys [<at 0.0038 0>,<at 0.0763 0.5 RCIM_Cubic>,<at 0.9957 0>]}
       BlueCurve {keys [<at 0 1 RCIM_Cubic>,<at 0.4894 3>,<at 0.6452 0.5906>,<at 1 0.5917 RCIM_Cubic>]}
       GreenCurve {keys [<at 0 0 RCIM_Cubic>,<at 0.4894 0>,<at 0.6452 0.1058>,<at 1 0 RCIM_Cubic>]}
       RedCurve {keys [<at 0 0.05 RCIM_Cubic>,<at 0.4894 0.2427>,<at 0.6452 0>,<at 1 0.0292 RCIM_Cubic>]}
      }})
    Particles.Variant1_ColorCurve:LinearColor=&ColorFromCurve(ColorCurve=DataInterface() {propertyValues {
       AlphaCurve {keys [<at 0.0039 0>,<at 0.0695 0.5>,<at 1 0>]}
       BlueCurve {keys [<at 0.2587 0.1912>,<at 0.5917 0.8417>,<at 0.9715 0.1912>]}
       GreenCurve {keys [<at 0.2587 0>,<at 0.5917 0.0667>,<at 0.9715 0>]}
       RedCurve {keys [<at 0.2587 0.0037>,<at 0.5917 0>,<at 0.9715 0.0037>]}
      }})
   ) {isSetParameters=true}
   &SkeletalMeshLocation(Bone Sampling Mode=::NewEnumerator4, MaskSpawnGroup=false, Orientation Sampling=::NewEnumerator1, Skeletal Mesh=$User.SkeletalMesh)
   &ParticleState()
   &UpdateMeshOrientation(Rotation Rate=/Game/Effects/Niagara/DynamicInputs/FN_PickFloatFromSpawnGroupCutoff(
     Over Cutoff Float:NiagaraFloat=0.05
     Under Cutoff Float:NiagaraFloat=-0.05
     SpawnGroup:NiagaraInt32=$Emitter.SpawnGroup
     Cutoff:NiagaraInt32=1
    ))
   &ScaleMeshSize(Scale Factor=&VectorFromFloat(Value=&FloatFromCurve(FloatCurve=DataInterface() {propertyValues {Curve {keys [
          {arriveTangent=-0.9629, interpMode=RCIM_Cubic, leaveTangent=-0.9629, tangentMode=RCTM_User, time=0, value=1}
          {arriveTangent=-1.4502, interpMode=RCIM_Cubic, leaveTangent=-1.4502, tangentMode=RCTM_User, time=1, value=0.4}
         ]}}})))
  ]
  rendererClasses [&NiagaraMeshRendererProperties]
  renderers [
   {rendererClass=&NiagaraMeshRendererProperties, rendererIndex=0}
   {rendererClass=&NiagaraMeshRendererProperties, rendererIndex=1}
  ]
 }
 SpinnyStatic001 {
  emitterUpdateScript [
   &EmitterState(Life Cycle Mode=::NewEnumerator1, Loop Behavior=::NewEnumerator1, Loop Delay=1.75, Loop Duration=1, UseLoopDelay=true)
   &SpawnBurst_Instantaneous(Spawn Count=2, Use Spawn Probability=false)
  ]
  particleSpawnScript [
   &InitializeParticle(
    Lifetime=0.75
    Mesh Scale Mode=::NewEnumerator3
    Mesh Uniform Scale=3
    Position Offset=<8 0 0>
    UsePositionOffset=true
    Write Material Random=true
   ) {module=InitializeParticle001}
   &InitialMeshOrientation(
    Orientation Axis=<0 0 1>
    Rotation=<0 0 0.25>
    Rotation Coordinate Space=::NewEnumerator0
    Use Orientation Vector=true
    Use Rotation Vector=true
   )
  ]
  particleUpdateScript [
   /SetVariables(
    Particles.Variant0_ColorCurve:LinearColor=&ColorFromCurve(ColorCurve=DataInterface() {propertyValues {
       AlphaCurve {keys [<at 0.0038 0>, {arriveTangent=1.8872, interpMode=RCIM_Cubic, leaveTangent=1.8872, time=0.0763, value=1}, <at 0.4012 0.75>, <at 0.9957 0>]}
       BlueCurve {keys [<at 0 1 RCIM_Cubic>,<at 0.4894 3>,<at 0.6452 0.5906>,<at 1 0.5917 RCIM_Cubic>]}
       GreenCurve {keys [<at 0 0 RCIM_Cubic>,<at 0.4894 0>,<at 0.6452 0.1058>,<at 1 0 RCIM_Cubic>]}
       RedCurve {keys [<at 0 0.05 RCIM_Cubic>,<at 0.4894 0.2427>,<at 0.6452 0>,<at 1 0.0292 RCIM_Cubic>]}
      }})
    Particles.Variant1_ColorCurve:LinearColor=&ColorFromCurve(ColorCurve=DataInterface() {propertyValues {
       AlphaCurve {keys [<at 0.0039 0>,<at 0.0695 1>,<at 1 0>]}
       BlueCurve {keys [<at 0.2587 0.1912>,<at 0.5917 0.8417>,<at 0.9715 0.1912>]}
       GreenCurve {keys [<at 0.2587 0>,<at 0.5917 0.0667>,<at 0.9715 0>]}
       RedCurve {keys [<at 0.2587 0.0037>,<at 0.5917 0>,<at 0.9715 0.0037>]}
      }})
   ) {isSetParameters=true}
   &ParticleState()
   &SkeletalMeshLocation(
    Bone Sampling Mode=::NewEnumerator4
    MaskSpawnGroup=false
    Orientation Sampling=::NewEnumerator1
    Sampled Position Offset=<0 15 0>
    Skeletal Mesh=$User.SkeletalMesh
    Velocity Sampling=::NewEnumerator0
    VelocitySpeedLimit=false
   )
   &ScaleMeshSize(Scale Factor=&VectorFromFloat(Value=&FloatFromCurve(FloatCurve=DataInterface() {propertyValues {Curve {keys [
          {arriveTangent=1.8128, interpMode=RCIM_Cubic, leaveTangent=1.8128, tangentMode=RCTM_User, time=0, value=0}
          {arriveTangent=0.7769, interpMode=RCIM_Cubic, leaveTangent=0.7769, tangentMode=RCTM_User, time=1, value=1}
         ]}}})))
   &ScaleColor(
    Scale Alpha=&FloatFromCurve(FloatCurve=DataInterface() {propertyValues {Curve {keys [{interpMode=RCIM_Cubic, tangentMode=RCTM_User, time=0, value=1}, {interpMode=RCIM_Cubic, tangentMode=RCTM_User, time=1, value=0}]}}})
    Scale RGB=&MakeVectorFromLinearColorRGB(Color=&ColorFromCurve(ColorCurve=DataInterface() {propertyValues {
        AlphaCurve {keys [<at 0 1 RCIM_Cubic>]}
        BlueCurve {keys [<at 0 1 RCIM_Cubic>,<at 0.4894 3>,<at 1 0.5917 RCIM_Cubic>]}
        GreenCurve {keys [<at 0 0.4192 RCIM_Cubic>,<at 0.4894 0>,<at 1 0 RCIM_Cubic>]}
        RedCurve {keys [<at 0 0.2412 RCIM_Cubic>,<at 0.4894 0.2427>,<at 1 0.0292 RCIM_Cubic>]}
       }}))
    ScaleA=true
    ScaleRGB=true
   )
  ]
  rendererClasses [&NiagaraMeshRendererProperties]
  renderers [
   {rendererClass=&NiagaraMeshRendererProperties, rendererIndex=0}
   {rendererClass=&NiagaraMeshRendererProperties, rendererIndex=1}
  ]
 }
}
summary {
 systemName=NS_Nebula_Glider_Deploy
 emitters [
  {emitterName=Rock_Moon_02, rendererClasses [&NiagaraMeshRendererProperties, &NiagaraSpriteRendererProperties]}
  {emitterName=Flames_Attached_to_Rocks007, rendererClasses [&NiagaraRibbonRendererProperties]}
  {emitterName=CenterGlow, rendererClasses [&NiagaraSpriteRendererProperties]}
  {emitterName=Flames_Attached_to_Rocks001, rendererClasses [&NiagaraRibbonRendererProperties]}
  {
   emitterName=Rock_Moon_02001
   rendererClasses [&NiagaraMeshRendererProperties, &NiagaraSpriteRendererProperties]
  }
  {emitterName=Glow, rendererClasses [&NiagaraSpriteRendererProperties]}
  {emitterName=Sparks, rendererClasses [&NiagaraSpriteRendererProperties]}
  {emitterName=SpinnyStatic, rendererClasses [&NiagaraMeshRendererProperties]}
  {emitterName=Body_Rings002, rendererClasses [&NiagaraMeshRendererProperties]}
  {emitterName=SpinnyStatic001, rendererClasses [&NiagaraMeshRendererProperties]}
 ]
 userVariables [
  User.ParaGlideLeanAlpha {type=NiagaraFloat, defaultValue=0, description=""}
  User.SkeletalMesh {
   type=NiagaraDataInterfaceSkeletalMesh
   defaultValue {
    struct=NiagaraExt_VariableValue_DataInterface
    value {dataInterface=&self:NiagaraDataInterfaceSkeletalMesh_0}
   }
  }
  User.VariantID {type=NiagaraFloat, defaultValue=0, description=""}
 ]
}
