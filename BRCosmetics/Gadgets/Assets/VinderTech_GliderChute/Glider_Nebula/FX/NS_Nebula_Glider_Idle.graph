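# Niagara system for the idle VFX of the Nebula glider cosmetic, with ribbon trails, body ring meshes, a flame mesh and glow sprites.
# Rock flames and their ribbons follow skeletal mesh sockets indexed by spawn group, while BigTrails uses five bone-pair anim trails.
# User.ForwardDot scales trail speed, system timelines drive scale and brightness, and modifiers adapt values for front-end and time of day.
# The Horn_Airglow emitter is disabled.
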NiagaraSystem /BRCosmetics/Gadgets/Assets/VinderTech_GliderChute/Glider_Nebula/FX/NS_Nebula_Glider_Idle
dependencies {
 usedDataInterfaces [
  &NiagaraDataInterfaceCurve
  &NiagaraDataInterfaceArrayFloat3
  &NiagaraDataInterfaceArrayFloat
  &NiagaraDataInterfaceColorCurve
  &NiagaraDataInterfaceVectorCurve
  &NiagaraDataInterfaceArrayColor
 ]
 usedDynamicInputs [
  &MakeFloatFromVector
  &SetBoolByFloatComparison
  &SelectLinearColorFromArray
  &MakeIntFromFloat
  &Abs_Float
  &Boolean_AndOperation
 ]
 usedModules [
  &self:SystemSpawnScript.SystemScriptSource.SystemScriptGraph.NiagaraNodeAssignment_0.SetVariables_#1
  &self:SystemSpawnScript.SystemScriptSource.SystemScriptGraph.NiagaraNodeAssignment_1.SetVariables_#2
  &SystemState
  &FN_Integrator
  &Timeline
 ]
}
emitters {
 BigTrails {
  emitterUpdateScript [
   /SetVariables(Emitter.VelocityTrail:Vector3f=&Add_Vector(A=&Multiply_VectorByFloat(Vector=$Engine.Owner.SystemYAxis, Float=300), B=$Engine.Owner.Velocity)) {enabled=false, isSetParameters=true}
   &EmitterState(Delay First Loop Only=true, Life Cycle Mode=::NewEnumerator1, Loop Delay=0.35, Loop Duration=0.1, UseLoopDelay=true)
   &FN_SpawnAnimTrail_v2(Number of Anim Trails=5)
  ]
  particleSpawnScript [
   /Game/Effects/Niagara/Modules/AnimTrail/FN_InitializeAnimTrail/FN_InitializeAnimTrailBoneIndecies(Second Bone or Socket Index=2)
   &FN_InitializeAnimTrailBoneIndecies(AnimTrail Selection=1, First Bone or Socket Index=1, Second Bone or Socket Index=3) {module=FN_InitializeAnimTrailBoneIndecies001}
   &FN_InitializeAnimTrailBoneIndecies(AnimTrail Selection=2, First Bone or Socket Index=2, Second Bone or Socket Index=4) {module=FN_InitializeAnimTrailBoneIndecies002}
   &FN_InitializeAnimTrailBoneIndecies(AnimTrail Selection=3, First Bone or Socket Index=3, Second Bone or Socket Index=5) {module=FN_InitializeAnimTrailBoneIndecies003}
   &FN_InitializeAnimTrailBoneIndecies(AnimTrail Selection=4, First Bone or Socket Index=4, Second Bone or Socket Index=6) {module=FN_InitializeAnimTrailBoneIndecies004}
   /Game/Effects/Niagara/Modules/AnimTrail/FN_InitializeAnimTrail/FN_InitializeAnimTrail(
    AnimTrailWidth=2
    Lifetime:NiagaraFloat=&FN_FrontEndFloatModifier(Gameplay=0.5, Front End=0.4)
    Write Material Random=false
   )
   &AddVelocity(Velocity=&FN_FrontEndVector3Modifier(
     Gameplay=&MakeVector(
      X=0
      Y=&FloatFromCurve(
       FloatCurve=DataInterface() {propertyValues {Curve {keys [<at 0 0.25>,<at 0.5 1>]}}}
       CurveIndex=&RemapRange(Input Value To Remap=$User.ForwardDot, Input Min=-1, Input Max=1, Output Min=0.5, Output Max=1)
       Scale Curve=500
      )
      Z=0
     )
     Front End=&MakeVector(
      X=0
      Y=&FloatFromCurve(
       FloatCurve=DataInterface() {propertyValues {Curve {keys [<at 0 0.25>,<at 0.5 1>]}}}
       CurveIndex=&RemapRange(Input Value To Remap=$User.ForwardDot, Input Min=-1, Input Max=1, Output Min=0, Output Max=1)
       Scale Curve=500
      )
      Z=75
     )
    ))
  ]
  particleUpdateScript [
   &InheritSourceMovement(
    Applied Position Scale=&MakeVector(
     X=&FloatFromCurve(FloatCurve=DataInterface() {propertyValues {Curve {keys [{interpMode=RCIM_Cubic, tangentMode=RCTM_User, time=0, value=1}, {interpMode=RCIM_Cubic, tangentMode=RCTM_User, time=1, value=0.3}]}}})
     Y=&FloatFromCurve(FloatCurve=DataInterface() {propertyValues {Curve {keys [{interpMode=RCIM_Cubic, tangentMode=RCTM_User, time=0, value=1}, {interpMode=RCIM_Cubic, tangentMode=RCTM_User, time=1, value=0.3}]}}})
     Z=1
    )
    ApplyForce=false
    ApplyPosition=true
    ApplyPositionFalloff=false
   )
   &AccelerationForce(
    Acceleration=&MakeVector(
     X=0
     Y=0
     Z=&MakeCustomFloatFromBool(
      Bool=$User.bIsFrontEnd
      True Float=&FloatFromCurve(FloatCurve=DataInterface() {propertyValues {Curve {keys [<at 0 -1>,<at 1 2.5>]}}}, Scale Curve=2000)
      False Float=0
     )
    )
    Coordinate Space=::Simulation
   )
   /SetVariables(Particles.Velocity=&Add_Vector(
     A=$Particles.Velocity
     B=&MakeCustomVectorFromBool(
      Bool=$User.bIsFrontEnd
      True Vector=&RotateVectorByQuaternion(
       Vector To Rotate=&MakeVector(
        X=&FloatFromCurve(FloatCurve=DataInterface() {propertyValues {Curve {keys [<at 0 -0.3>,<at 1 0.8>]}}}, Scale Curve=200)
        Y=0
        Z=0
       )
       Quaternion=$Engine.Owner.Rotation
      )
      False Vector=<1 0 0>
     )
    )) {enabled=false, isSetParameters=true}
   &SolveForcesAndVelocity(Clamp Velocity=false, Limit Acceleration=false)
   &FN_UpdateAnimTrail_v2()
   &ScaleRibbonWidth()
   &FN_RibbonMaterial() {enabled=false}
   &ScaleColor(
    Scale Alpha=&FloatFromCurve(
     FloatCurve=DataInterface() {propertyValues {Curve {keys [{interpMode=RCIM_Cubic, tangentMode=RCTM_User, time=0, value=1}, {interpMode=RCIM_Cubic, tangentMode=RCTM_User, time=1, value=0}]}}}
     Scale Curve=&FloatFromCurve(
      FloatCurve=DataInterface() {propertyValues {Curve {keys [
          {interpMode=RCIM_Cubic, tangentMode=RCTM_User, time=0, value=1}
          {interpMode=RCIM_Cubic, tangentMode=RCTM_User, time=0.5, value=2}
          {tangentMode=RCTM_User, time=1, value=1.25}
         ]}}}
      CurveIndex=$System.Timeline001.NormalizedTime
      Scale Curve=1.25
     )
    )
    Scale RGB=&VectorFromFloat(Value=&FNTimeOfDayScalar(Night Scale=1, Morning Scale=1, Day Scale=2, Evening Scale=1))
    ScaleA=true
    ScaleRGB=true
   )
   /SetVariables(Particles.Renderer2_RibbonWidth:NiagaraFloat=&Multiply_Float(A=$Particles.RibbonWidth, B=0.9)) {isSetParameters=true}
   /SetVariables(Particles.Renderer2Position:Vector3f=&Add_Vector(A=Linked() {linkedVariable {name=Particles.Position, type=NiagaraPosition}}, B=<0 0 -5>)) {isSetParameters=true}
   &DynamicMaterialParameters(
    Index 0 Param 1=/Game/Effects/Niagara/DynamicInputs/Random/FN_SeededRandomRangeFloat(Minimum:NiagaraFloat=0, Maximum:NiagaraFloat=1, Random Seed:NiagaraInt32=$Particles.SpawnGroup)
    Module.Param0WriteEnabled=true
    Module.Param1WriteEnabled=true
    Module.Param2WriteEnabled=true
    Module.Param3WriteEnabled=true
   )
  ]
  rendererClasses [&NiagaraRibbonRendererProperties]
  renderers [
   {rendererClass=&NiagaraRibbonRendererProperties}
   {rendererClass=&NiagaraRibbonRendererProperties, rendererIndex=1}
  ]
 }
 Body_Rings002 {
  emitterUpdateScript [
   &EmitterState(Life Cycle Mode=::NewEnumerator1, Loop Behavior=::NewEnumerator1, Loop Delay=0.1, Loop Duration Mode=::NewEnumerator1, UseLoopDelay=true)
   &SpawnBurst_Instantaneous(Use Spawn Probability=false, UseLoopCountLimit=false)
   &SpawnRate(SpawnRate=&Lerp_Float(A=1.5, B=2, Alpha=$System.Timeline001.NormalizedTime), Use Spawn Probability=false)
   &SpawnBurst_Instantaneous(Spawn Time=1.3653, Use Spawn Probability=false, UseLoopCountLimit=false) {module=SpawnBurst_Instantaneous001}
  ]
  particleSpawnScript [&InitializeParticle(Lifetime=3.5, Mesh Scale Mode=::NewEnumerator3, UsePositionOffset=true, Write Material Random=true)]
  particleUpdateScript [
   /SetVariables(
    Particles.Variant0_ColorCurve:LinearColor=&MultiplyLinearColor(
     Linear Color=&ColorFromCurve(ColorCurve=DataInterface() {propertyValues {
        AlphaCurve {keys [<at 0 1>,<at 0.6097 1>,<at 0.76 1>,<at 1 0>]}
        BlueCurve {keys [<at 0 0.1294>,<at 0.3539 0.3087>,<at 0.7059 1>]}
        GreenCurve {keys [<at 0 0>,<at 0.3539 0>,<at 0.7059 0.0976>]}
        RedCurve {keys [<at 0 0.0323>,<at 0.3539 0.0085>,<at 0.7059 0.0482>]}
       }})
     Scale Factor=&MakeVector4(X=1, Y=1, Z=1, W=&FNTimeOfDayScalar(Night Scale=0.18, Morning Scale=1, Evening Scale=0.18))
    )
    Particles.Variant1_ColorCurve:LinearColor=&MultiplyLinearColor(
     Linear Color=&ColorFromCurve(ColorCurve=DataInterface() {propertyValues {
        AlphaCurve {keys [<at 0.0039 0>,<at 0.0695 1>,<at 0.7033 1>,<at 0.7791 1>,<at 1 0>]}
        BlueCurve {keys [<at 0.2587 0.1912>,<at 0.5917 0.8417>,<at 0.9715 0.1912>]}
        GreenCurve {keys [<at 0.2587 0>,<at 0.5917 0.0667>,<at 0.9715 0>]}
        RedCurve {keys [<at 0.2587 0.0037>,<at 0.5917 0>,<at 0.9715 0.0037>]}
       }})
     Scale Factor=&MakeVector4(X=1, Y=1, Z=1, W=&FNTimeOfDayScalar(Night Scale=0.2, Morning Scale=0.5, Evening Scale=0.3))
    )
   ) {isSetParameters=true}
   &ParticleState()
   &SkeletalMeshLocation(Bone Sampling Mode=::NewEnumerator4, MaskSpawnGroup=false, Skeletal Mesh=$User.GliderMesh)
   &ScaleMeshSize(Scale Factor=&VectorFromFloat(Value=&FloatFromCurve(
      FloatCurve=DataInterface() {propertyValues {Curve {keys [
          {arriveTangent=-0.5446, interpMode=RCIM_Cubic, leaveTangent=-0.5446, tangentMode=RCTM_User, time=0, value=0.25}
          {arriveTangent=0.9755, interpMode=RCIM_Cubic, leaveTangent=0.9755, tangentMode=RCTM_User, time=0.4, value=0.7}
          {arriveTangent=0.6055, interpMode=RCIM_Cubic, leaveTangent=0.6055, tangentMode=RCTM_User, time=0.7, value=0.95}
          <at 1 1.15 RCIM_Cubic>
         ]}}}
      Scale Curve=1.25
     ))) {module=ScaleMeshSize001}
   &DynamicMaterialParameters(
    Index 2 Param 1=&RandomRangeFloat(Minimum=0.8, Maximum=1.5)
    Index 2 Param 2=&RandomRangeFloat(Minimum=0.25, Maximum=0.5)
    Index 2 Param 3=-0.1
    Index 2 Param 4=0.1
    Module.Param0WriteEnabled=true
    Module.Param0WriteEnabled2=true
    Module.Param1WriteEnabled=true
    Module.Param1WriteEnabled2=true
    Module.Param2WriteEnabled=true
    Module.Param2WriteEnabled2=true
    Module.Param3WriteEnabled=true
    Module.Param3WriteEnabled2=true
    Write Parameter Index 2=true
   ) {module=DynamicMaterialParameters001}
   &ScaleMeshSize(Scale Factor=&VectorFromCurve(
     VectorCurve=DataInterface() {propertyValues {
       XCurve {keys [{tangentMode=RCTM_User, time=0, value=0}, {tangentMode=RCTM_User, time=0.5, value=0}, <at 1 1 RCIM_Cubic>]}
       YCurve {keys [{tangentMode=RCTM_User, time=0, value=0}, {tangentMode=RCTM_User, time=0.5, value=0}, <at 1 1 RCIM_Cubic>]}
       ZCurve {keys [{tangentMode=RCTM_User, time=0, value=0}, {tangentMode=RCTM_User, time=0.5, value=0}, <at 1 1 RCIM_Cubic>]}
      }}
     CurveIndex=$System.Timeline.NormalizedTime
    )) {enabled=false}
   &ScaleColor(
    Linear Color Curve=DataInterface() {propertyValues {
      AlphaCurve {keys [<at 0 1>,<at 0.5247 1>,<at 1 0>]}
      BlueCurve {keys [<at 0 0.1294>,<at 0.3539 0.3087>,<at 0.7059 1>]}
      GreenCurve {keys [<at 0 0>,<at 0.3539 0>,<at 0.7059 0.0976>]}
      RedCurve {keys [<at 0 0.0323>,<at 0.3539 0.0085>,<at 0.7059 0.0482>]}
     }}
    Scale Mode=::NewEnumerator2
   ) {enabled=false}
   &ScaleColor(
    Scale RGB=&VectorFromFloat(Value=&FNTimeOfDayScalar(Night Scale=0.2, Morning Scale=1, Evening Scale=0.2))
    ScaleA=true
    ScaleRGB=true
   ) {module=ScaleColor001}
   &SolveForcesAndVelocity(Clamp Velocity=false, Limit Acceleration=false)
  ]
  rendererClasses [&NiagaraMeshRendererProperties]
  renderers [
   {rendererClass=&NiagaraMeshRendererProperties, rendererIndex=0}
   {rendererClass=&NiagaraMeshRendererProperties, rendererIndex=1}
  ]
 }
 CenterGlow {
  emitterUpdateScript [
   &EmitterState(Life Cycle Mode=::NewEnumerator1, Loop Behavior=::NewEnumerator1, Loop Delay=0.35, Loop Duration Mode=::NewEnumerator1, UseLoopDelay=true)
   &SpawnBurst_Instantaneous(Spawn Count=2, Use Spawn Probability=false, UseLoopCountLimit=false)
  ]
  particleSpawnScript [&InitializeParticle(
    Color=$System.GlowColor
    Lifetime=1.5
    Position Offset=&FN_FrontEndVector3Modifier(Gameplay=<0 30 -50>, Front End=<0 20 -50>)
    Sprite Size=<140 85>
    Sprite Size Mode=::NewEnumerator4
    UsePositionOffset=true
    Write Material Random=true
   )]
  particleUpdateScript [
   &ParticleState(
    Kill Particles When Lifetime Has Elapsed=false
    Let Infinitely Lived Particles Die When Emitter Deactivates=true
    Loop Particles Lifetime=true
   )
   &SolveForcesAndVelocity(Clamp Velocity=false, Limit Acceleration=false)
   &ScaleColor(
    Scale Alpha=&FloatFromCurve(
     FloatCurve=DataInterface() {propertyValues {Curve {keys [{tangentMode=RCTM_User, time=0, value=0}, {tangentMode=RCTM_User, time=1, value=1}]}}}
     CurveIndex=$System.Timeline.NormalizedTime
     Scale Curve=&FNTimeOfDayScalar(Night Scale=0.4, Morning Scale=0.75, Evening Scale=0.4)
    )
    ScaleA=true
    ScaleRGB=true
   )
   &FN_Light_Attributes(Create Color Attribute=true, Create Exponent Attribute=true, Create Radius Attribute=true, Radius=35)
   &CameraOffset(Camera Offset Amount=5)
  ]
  rendererClasses [&NiagaraSpriteRendererProperties]
  renderers [{rendererClass=&NiagaraSpriteRendererProperties}]
 }
 FlameMesh001 {
  emitterUpdateScript [
   &EmitterState(
    Inactive Response=::NewEnumerator1
    Life Cycle Mode=::NewEnumerator1
    Loop Behavior=::NewEnumerator1
    Loop Delay=0.35
    Loop Duration Mode=::NewEnumerator1
    Reset Age On Awaken=true
    Scalability Mode=::NewEnumerator1
    UseLoopDelay=true
   )
   &SpawnBurst_Instantaneous(Spawn Count=1, Spawn Time=0.15, Use Spawn Probability=false)
  ]
  particleSpawnScript [
   /SetVariables(Particles.InnerRing_Color:LinearColor=&MultiplyLinearColor(
     Linear Color=<rgba 0.0621 0.0983 0.7945 1>
     Scale Factor=&MakeVector4(X=0, Y=0, Z=0, W=&FNTimeOfDayScalar(Night Scale=0.5, Morning Scale=1, Evening Scale=0.5))
    )) {enabled=false, isSetParameters=true}
   /SetVariables(Particles.Cone_Pos:NiagaraPosition=<0 -225 -85>) {isSetParameters=true}
   /Niagara/Modules/Spawn/Initialization/InitializeParticle(
    Color=$System.BigRingColor
    Mesh Scale=<0.76 1 1>
    Module.Write Color=true
    Module.Write Lifetime=true
    Module.Write Mass=true
    Module.Write Material Random=true
    Module.Write Position=false
    Module.Write Scale=true
    Module.Write SpriteRotation=false
    Module.Write SpriteSize=false
    Position=&SimulationPosition()
    Sprite Size=<5 5>
   )
   &DynamicMaterialParameters(
    Index 0 Param 1=&RandomRangeFloat(Minimum=0, Maximum=1)
    Index 0 Param 2=2
    Index 0 Param 3=&FN_FrontEndFloatModifier(Gameplay=20, Front End=12)
    Module.Param0WriteEnabled=true
    Module.Param1WriteEnabled=true
    Module.Param2WriteEnabled=true
    Module.Param3WriteEnabled=true
   )
   &InitialMeshOrientation(
    Mesh Orientation Mode=::NewEnumerator2
    Rotation=&FN_FrontEndVector3Modifier(Gameplay=<0 0 0>, Front End=<0 0 0>)
    Use Rotation Vector=true
   )
   &SystemLocation(MaskSpawnGroup=false)
  ]
  particleUpdateScript [
   &ParticleState(Kill Particles When Lifetime Has Elapsed=false, Let Infinitely Lived Particles Die When Emitter Deactivates=true)
   /SetVariables(
    Particles.Variant0_ConeColor:LinearColor=&MultiplyLinearColor(
     Linear Color=<rgba 1 1 1 1>
     Scale Factor=&MakeVector4(
      X=1
      Y=1
      Z=1
      W=&Multiply_Float(
       A=&Lerp_Float(A=0.5, B=1, Alpha=$System.Timeline001.NormalizedTime)
       B=&FNTimeOfDayScalar(Night Scale=0.3, Morning Scale=1, Evening Scale=0.5)
      )
     )
    )
    Particles.Variant1_ConeColor:LinearColor=&MultiplyLinearColor(
     Linear Color=<rgba 1 1 1 1>
     Scale Factor=&MakeVector4(
      X=1
      Y=1
      Z=1
      W=&Multiply_Float(
       A=&Lerp_Float(A=0.5, B=1, Alpha=$System.Timeline001.NormalizedTime)
       B=&FNTimeOfDayScalar(Night Scale=0.3, Morning Scale=1, Evening Scale=0.5)
      )
     )
    )
   ) {isSetParameters=true}
   &InterpolateOverTime(
    Rate Of Change:NiagaraFloat=$User.Interpolate_Rate_of_Change
    Target Vector Value:NiagaraPosition=$Particles.Position
    Variable Type=::NewEnumerator2
   )
   /SetVariables(Particles.Interpolate Dynamic=&MakeFloat4FromVectorAndFloat("Vector (RGB)"=$Output.InterpolateOverTime.Moving Average, "Float (Alpha)"=0)) {isSetParameters=true}
   &DynamicMaterialParameters(
    Index 2 Param 1=0.34
    Index 2 Param 2=1
    Index 2 Param 3=-0.23
    Index 2 Param 4=0.15
    Module.Param0WriteEnabled=true
    Module.Param0WriteEnabled2=true
    Module.Param1WriteEnabled=true
    Module.Param1WriteEnabled2=true
    Module.Param2WriteEnabled=true
    Module.Param2WriteEnabled2=true
    Module.Param3WriteEnabled=true
    Module.Param3WriteEnabled2=true
    Write Parameter Index 2=true
   ) {module=DynamicMaterialParameters001}
   &SolveForcesAndVelocity(Clamp Velocity=false, Limit Acceleration=false)
   &CameraOffset()
   &ScaleColor(
    Scale Alpha=&FloatFromCurve(FloatCurve=DataInterface() {propertyValues {Curve {keys [<at 0.2 0>,<at 0.4 1>]}}}, CurveIndex=$System.Age)
    Scale RGB=&VectorFromFloat(Value=&FloatFromCurve(
      FloatCurve=DataInterface() {propertyValues {Curve {keys [
          {interpMode=RCIM_Cubic, tangentMode=RCTM_User, time=0.2, value=0.2}
          {interpMode=RCIM_Cubic, tangentMode=RCTM_User, time=0.25, value=3}
          <at 0.5 1>
         ]}}}
      CurveIndex=$System.Age
     ))
    ScaleA=true
    ScaleRGB=true
   ) {enabled=false}
   &ScaleColor(
    Scale Alpha=&FloatFromCurve(
     FloatCurve=DataInterface() {propertyValues {Curve {keys [
         {interpMode=RCIM_Cubic, tangentMode=RCTM_User, time=0, value=1}
         {interpMode=RCIM_Cubic, tangentMode=RCTM_User, time=0.5, value=2}
         {tangentMode=RCTM_User, time=1, value=1.25}
        ]}}}
     CurveIndex=$System.Timeline001.NormalizedTime
     Scale Curve=&FN_FrontEndFloatModifier(Gameplay=1.3, Front End=2)
    )
    Scale RGB=&VectorFromFloat(Value=&FNTimeOfDayScalar(Night Scale=0.2, Morning Scale=0.5, Evening Scale=0.2))
    ScaleA=true
    ScaleRGB=true
   ) {module=ScaleColor001}
   &ScaleMeshSize(Scale Factor=&VectorFromFloat(Value=&FloatFromCurve(
      FloatCurve=DataInterface() {propertyValues {Curve {keys [
          {arriveTangent=9.0725, interpMode=RCIM_Cubic, leaveTangent=9.0725, tangentMode=RCTM_User, time=0, value=0}
          {arriveTangent=-2.5249, interpMode=RCIM_Cubic, leaveTangent=-2.5249, tangentMode=RCTM_User, time=0.1408, value=0.9847}
          {arriveTangent=-2.1394, interpMode=RCIM_Cubic, leaveTangent=-2.1394, tangentMode=RCTM_User, time=0.5, value=1}
         ]}}}
      CurveIndex=$System.Timeline.NormalizedTime
     )))
   &FN_Light_Attributes(
    Color=<rgba 0.8333 5.3743 20 1>
    Create Color Attribute=true
    Create Exponent Attribute=true
    Create Position Attribute=true
    Create Radius Attribute=true
    Exponent=2
    Position=&AddVectorToPosition(Position=$Particles.Position, Vector=<0 0 30>)
    Radius=70
   )
  ]
  rendererClasses [&NiagaraMeshRendererProperties]
  renderers [
   {rendererClass=&NiagaraMeshRendererProperties, rendererIndex=0}
   {rendererClass=&NiagaraMeshRendererProperties, rendererIndex=1}
   {rendererClass=&NiagaraMeshRendererProperties, rendererIndex=2}
   {rendererClass=&NiagaraMeshRendererProperties, rendererIndex=3}
  ]
 }
 Flames_Attached_to_Rocks {
  emitterUpdateScript [
   &EmitterState(Life Cycle Mode=::NewEnumerator1, Loop Behavior=::NewEnumerator1, Loop Delay=0.35, Loop Duration Mode=::NewEnumerator1, UseLoopDelay=true)
   &SpawnBurst_Instantaneous(Spawn Count=4, Use Spawn Probability=false, UseLoopCountLimit=false)
  ]
  particleSpawnScript [
   /SetVariables(Particles.Sprite_Color:LinearColor=&MultiplyLinearColor(
     Linear Color=$System.RockGlowColor
     Scale Factor=&MakeVector4(X=1, Y=1, Z=1, W=&FNTimeOfDayScalar(Night Scale=0.5, Morning Scale=0.5, Evening Scale=0.5))
    )) {isSetParameters=true}
   &FN_SetSpawnGroup()
   /Niagara/Modules/Spawn/Initialization/V2/InitializeParticle(
    Lifetime=0.5
    Mesh Scale Mode=::NewEnumerator1
    Mesh Uniform Scale Max=0.9
    Mesh Uniform Scale Min=0.5
    UsePositionOffset=false
    Write Material Random=true
   )
   &InitialMeshOrientation(Mesh Orientation Mode=::NewEnumerator2, Rotation Coordinate Space=::NewEnumerator2, Use Rotation Vector=false)
  ]
  particleUpdateScript [
   &ParticleState(Kill Particles When Lifetime Has Elapsed=false)
   &SkeletalMeshLocation(
    Bone / Socket Index=$Particles.SpawnGroup
    Bone Sampling Mode=::NewEnumerator4
    MaskSpawnGroup=false
    Orientation Sampling=::NewEnumerator1
    Sampled Orientation Coordinate Space=::Local
    Sampled Position Offset=<0 -22 0>
    Sampled Position Offset Coordinate Space:ENiagaraSkelMeshTransforms=::NewEnumerator2
    Skeletal Mesh=$User.SkeletalMesh
   ) {module=SkeletalMeshLocation001}
   &UpdateMeshOrientation(Rotation Rate=&RandomRangeFloat(Minimum=-0.25, Maximum=0.25), Rotation Vector=<0 1 0>)
   /SetVariables(Particles.Sprite_Position:NiagaraPosition=$Output.SkeletalMeshLocation001.SampledPositionWithoutOffset) {isSetParameters=true}
   &SolveForcesAndVelocity(Clamp Velocity=false, Limit Acceleration=false)
   &DynamicMaterialParameters(
    Index 0 Param 2=0.5
    Index 1 Param 2=3
    Module.Param0WriteEnabled=true
    Module.Param0WriteEnabled1=true
    Module.Param1WriteEnabled=true
    Module.Param1WriteEnabled1=true
    Module.Param2WriteEnabled=true
    Module.Param2WriteEnabled1=true
    Module.Param3WriteEnabled=true
    Module.Param3WriteEnabled1=true
    Write Parameter Index 1=true
   ) {module=DynamicMaterialParameters001}
   &ScaleColor(
    Scale Alpha=&Multiply_Float(
     A=&Lerp_Float(A=&RandomRangeFloat(Minimum=0.25, Maximum=0.4), B=0.5, Alpha=$System.Timeline001.NormalizedTime)
     B=&FNTimeOfDayScalar(Night Scale=0.3, Morning Scale=1, Evening Scale=0.5)
    )
    ScaleA=true
    ScaleRGB=true
   )
  ]
  rendererClasses [&NiagaraMeshRendererProperties, &NiagaraSpriteRendererProperties]
  renderers [
   {rendererClass=&NiagaraMeshRendererProperties, rendererIndex=0}
   {rendererClass=&NiagaraSpriteRendererProperties, rendererIndex=1}
   {rendererClass=&NiagaraMeshRendererProperties, rendererIndex=2}
  ]
 }
 Flames_Attached_to_Rocks001 {
  emitterUpdateScript [
   &EmitterState()
   &SpawnBurst_Instantaneous(Spawn Count=4, Spawn Time=0.35, Use Spawn Probability=false, UseLoopCountLimit=false)
   &SpawnRate(SpawnRate=350, Use Spawn Probability=false)
  ]
  particleSpawnScript [
   &InitializeParticle(
    Color=<rgba 1 1 1 1.5>
    Lifetime=&FN_FrontEndFloatModifier(Gameplay=0.5, Front End=0.4)
    Ribbon Width=30
    Ribbon Width Mode=::NewEnumerator1
    Sprite Size Mode=::NewEnumerator3
    Uniform Sprite Size=55
    UsePositionOffset=false
    Write Material Random=true
   )
   &FN_SetSpawnGroup()
   &SkeletalMeshLocation(
    Bone / Socket Index=$Particles.SpawnGroup
    Bone Sampling Mode=::NewEnumerator4
    MaskSpawnGroup=false
    Sampled Position Offset=<0 -5 0>
    Sampled Position Offset Coordinate Space:ENiagaraSkelMeshTransforms=::NewEnumerator2
    Skeletal Mesh=$User.SkeletalMesh
   )
   &AddVelocity(Velocity=&MakeVector(
     X=0
     Y=&FloatFromCurve(
      FloatCurve=DataInterface() {propertyValues {Curve {keys [<at 0 0.25>,<at 0.5 1>]}}}
      CurveIndex=&RemapRange(Input Value To Remap=$User.ForwardDot, Input Min=-1, Input Max=1, Output Min=0.1, Output Max=1)
      Scale Curve=&MakeCustomFloatFromBool(
       Bool=&SetBoolByIntComparison(A=&SpawnGroup(), B=3, Comparison Type=::NewEnumerator0)
       True Float=450
       False Float=250
      )
     )
     Z=0
    )) {module=AddVelocity001}
   &AddVelocity(Velocity=&Lerp_Vector(
     A=<0 0 250>
     B=&SelectVectorFromArray(
      Direct Array Index=$Particles.SpawnGroup
      Vector Selection Array=DataInterface() {propertyValues {FloatData [<100 150 500>, <-100 150 500>]}}
     )
     Alpha=$System.Dive
    )) {enabled=false}
  ]
  particleUpdateScript [
   &InheritSourceMovement(
    Applied Position Falloff Curve=DataInterface() {propertyValues {Curve {keys [<at 0 1>,<at 0.5 0>]}}}
    Applied Position Falloff Distance=5
    ApplyForce=false
    ApplyPosition=true
    ApplyPositionFalloff=true
    ApplyPositionFalloffCurve=false
   ) {module=InheritSourceMovement001}
   &SolveForcesAndVelocity(Clamp Velocity=false, Limit Acceleration=false)
   &DynamicMaterialParameters(
    Index 0 Param 1=&SelectFloatFromArray(Direct Array Index=$Particles.SpawnGroup, Float Selection Array=DataInterface() {propertyValues {FloatData [0.8,0.4,0.2,1]}})
    Index 0 Param 2=$System.Dive
    Index 0 Param 3=3
    Module.Param0WriteEnabled=true
    Module.Param1WriteEnabled=true
    Module.Param2WriteEnabled=true
    Module.Param3WriteEnabled=true
   )
   &ParticleState()
   &ScaleColor(
    Scale Alpha=&FloatFromCurve(
     FloatCurve=DataInterface() {propertyValues {Curve {keys [{interpMode=RCIM_Cubic, tangentMode=RCTM_User, time=0, value=1}, {interpMode=RCIM_Cubic, tangentMode=RCTM_User, time=1, value=0}]}}}
     Scale Curve=&FloatFromCurve(
      FloatCurve=DataInterface() {propertyValues {Curve {keys [
          {interpMode=RCIM_Cubic, tangentMode=RCTM_User, time=0, value=1}
          {interpMode=RCIM_Cubic, tangentMode=RCTM_User, time=0.5, value=2}
          {tangentMode=RCTM_User, time=1, value=1}
         ]}}}
      CurveIndex=$System.Timeline001.NormalizedTime
      Scale Curve=1.25
     )
    )
    Scale RGB=&VectorFromFloat(Value=1)
    ScaleA=true
    ScaleRGB=true
   )
   &ScaleRibbonWidth(Ribbon Width Scale=&SelectFloatFromArray(Direct Array Index=$Particles.SpawnGroup, Float Selection Array=DataInterface() {propertyValues {FloatData [1,1.5,1.2,1]}}))
  ]
  rendererClasses [&NiagaraRibbonRendererProperties]
  renderers [
   {rendererClass=&NiagaraRibbonRendererProperties}
   {rendererClass=&NiagaraRibbonRendererProperties, rendererIndex=1}
  ]
 }
 Glow_Burst {
  emitterUpdateScript [
   &EmitterState(
    Inactive Response=::NewEnumerator1
    Life Cycle Mode=::NewEnumerator1
    Loop Behavior=::NewEnumerator1
    Loop Delay=0.25
    Loop Duration=0.5
    UseLoopDelay=true
   )
   &SpawnBurst_Instantaneous(Use Spawn Probability=false, UseLoopCountLimit=false)
  ]
  particleSpawnScript [&InitializeParticle(
    Color=$System.GlowColor
    Lifetime=0.5
    Sprite Rotation Mode=::NewEnumerator2
    Sprite Size=<250 250>
    Sprite Size Mode=::NewEnumerator4
    UsePositionOffset=false
    Write Material Random=true
   )]
  particleUpdateScript [
   &ParticleState()
   &ScaleColor(
    Scale Alpha=&FloatFromCurve(
     FloatCurve=DataInterface() {propertyValues {Curve {keys [
         {arriveTangent=9.0725, interpMode=RCIM_Cubic, leaveTangent=9.0725, tangentMode=RCTM_User, time=0, value=0}
         {arriveTangent=-2.5249, interpMode=RCIM_Cubic, leaveTangent=-2.5249, tangentMode=RCTM_User, time=0.1408, value=0.9847}
         {arriveTangent=-0.0405, interpMode=RCIM_Cubic, leaveTangent=-0.0405, tangentMode=RCTM_User, time=1, value=0}
        ]}}}
     Scale Curve=0.25
    )
    ScaleA=true
    ScaleRGB=true
   )
   &CameraOffset(Camera Offset Amount=100)
  ]
  rendererClasses [&NiagaraSpriteRendererProperties]
  renderers [
   {rendererClass=&NiagaraSpriteRendererProperties}
   {rendererClass=&NiagaraSpriteRendererProperties, rendererIndex=1}
  ]
 }
 Horn_Airglow {
  bEnabled=false
  emitterUpdateScript [
   &EmitterState(Life Cycle Mode=::NewEnumerator1, Loop Behavior=::NewEnumerator1, Loop Duration Mode=::NewEnumerator1, UseLoopDelay=false)
   &SpawnBurst_Instantaneous(Spawn Count=2, Use Spawn Probability=false)
  ]
  particleSpawnScript [
   &PartitionParticles()
   /SetVariables(Particles.BladeGlowPosition:NiagaraPosition=<-25 -1.5 -3>) {enabled=false, isSetParameters=true}
   /Niagara/Modules/Spawn/Initialization/InitializeParticle(
    Color=<rgba 0 0.9302 1 0.25>
    Mesh Scale=<100 200 0>
    Module.Write Color=true
    Module.Write Lifetime=true
    Module.Write Mass=true
    Module.Write Material Random=true
    Module.Write Position=false
    Module.Write Scale=false
    Module.Write SpriteRotation=false
    Module.Write SpriteSize=true
    Position=&SimulationPosition()
    Sprite Size=<35 75>
   )
   &SpriteFacingAndAlignment(Alignment Coordinate Space=::Local, Sprite Alignment=<0 -0.5 1>, Sprite Facing=<0 0 1>, WriteAlignment=true, WriteFacing=false)
  ]
  particleUpdateScript [
   &ParticleState(Kill Particles When Lifetime Has Elapsed=false)
   &SkeletalMeshLocation(
    Bone / Socket Index=&MakeCustomIntFromBool(
     Bool=&SetBoolByIntComparison(A=$Particles.PartitionParticles.Partition, B=0, Comparison Type=::NewEnumerator0)
     True INT=4
     False INT=5
    )
    Bone Sampling Mode=::NewEnumerator4
    Inherit Velocity=false
    Interpolate Mesh Sampling=false
    MaskSpawnGroup=false
    Sampled Position Offset Coordinate Space:ENiagaraMeshTransforms=::NewEnumerator2
    Skeletal Mesh=$User.SkeletalMesh
   )
   &ScaleColor(
    Scale Alpha=&FloatFromCurve(
     FloatCurve=DataInterface() {propertyValues {Curve {keys [{interpMode=RCIM_Cubic, tangentMode=RCTM_User, time=0, value=0}, {interpMode=RCIM_Cubic, tangentMode=RCTM_User, time=0.5, value=0.5}]}}}
     CurveIndex=$Emitter.Age
     Scale Curve=0.45
    )
    Scale RGB=&VectorFromFloat(Value=0.15)
    ScaleA=true
    ScaleRGB=true
   ) {enabled=false}
   &CameraOffset(Camera Offset Amount=20) {enabled=false}
  ]
  rendererClasses [&NiagaraMeshRendererProperties, &NiagaraSpriteRendererProperties]
  renderers [
   {rendererClass=&NiagaraMeshRendererProperties, rendererIndex=0}
   {rendererClass=&NiagaraSpriteRendererProperties, rendererIndex=1}
  ]
 }
}
summary {
 systemName=NS_Nebula_Glider_Idle
 emitters [
  {emitterName=Flames_Attached_to_Rocks001, rendererClasses [&NiagaraRibbonRendererProperties]}
  {emitterName=BigTrails, rendererClasses [&NiagaraRibbonRendererProperties]}
  {emitterName=FlameMesh001, rendererClasses [&NiagaraMeshRendererProperties]}
  {emitterName=Body_Rings002, rendererClasses [&NiagaraMeshRendererProperties]}
  {emitterName=CenterGlow, rendererClasses [&NiagaraSpriteRendererProperties]}
  {
   emitterName=Flames_Attached_to_Rocks
   rendererClasses [&NiagaraMeshRendererProperties, &NiagaraSpriteRendererProperties]
  }
  {emitterName=Glow_Burst, rendererClasses [&NiagaraSpriteRendererProperties]}
  {
   bEnabled=false
   emitterName=Horn_Airglow
   rendererClasses [&NiagaraMeshRendererProperties, &NiagaraSpriteRendererProperties]
  }
 ]
 userVariables [
  User.bIsFrontEnd {type=NiagaraBool}
  User.bIsFrontEndPreview {type=NiagaraBool}
  User.bIsFullyDeployed {type=NiagaraBool}
  User.blsFullyDeployed {type=NiagaraBool, defaultValue=true}
  User.ForwardDot {type=NiagaraFloat, defaultValue=1, description=""}
  User.GliderMesh {
   type=NiagaraDataInterfaceSkeletalMesh
   defaultValue {
    struct=NiagaraExt_VariableValue_DataInterface
    value {dataInterface=&self:NiagaraDataInterfaceSkeletalMesh_0}
   }
  }
  User.Interpolate_Rate_of_Change {type=NiagaraFloat, defaultValue=0, description=""}
  User.New Vector {type=Vector3f, description="", defaultValue {struct=Vector3f, value=<0 0 0>}}
  User.ParaGlideLeanAlpha {type=NiagaraFloat, defaultValue=0, description=""}
  User.RibbonIDOffset {type=NiagaraInt32}
  User.RightDot {type=NiagaraFloat, defaultValue=0, description=""}
  User.RotationPreview {type=Vector3f, description="", defaultValue {struct=Vector3f, value=<0 0 0>}}
  User.Secondary Color {type=LinearColor, description="", defaultValue {struct=LinearColor, value=<rgba 1 1 1 1>}}
  User.SkeletalMesh {
   type=NiagaraDataInterfaceSkeletalMesh
   defaultValue {
    struct=NiagaraExt_VariableValue_DataInterface
    value {dataInterface=&self:NiagaraDataInterfaceSkeletalMesh_1}
   }
  }
  User.SkeletalMeshNebula {
   type=NiagaraDataInterfaceSkeletalMesh
   defaultValue {
    struct=NiagaraExt_VariableValue_DataInterface
    value {dataInterface=&self:NiagaraDataInterfaceSkeletalMesh_2}
   }
  }
  User.VariantID {type=NiagaraFloat, defaultValue=0, description=""}
 ]
}
